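# Unlit masked material for glowstick mesh particles in the JadeTowel skydiving contrail Niagara effect.
# The green channel of a light FX mask texture blends between a dimmed and full GlowstickColor tint, which feeds a shared emissive adjustment function.
# A DeriveHDRfromLDR function also takes the color and an HDR intensity parameter, though no link from it to the emissive chain appears.
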Material /BRCosmetics/Athena/Effects/Particles/Skydiving/NS_Contrail_JadeTowel/M_Contrail_JadeTowel_Glowstick
graph {
 properties {
  bAutomaticallySetUsageInEditor=false
  bUsedWithNiagaraMeshParticles=true
  blendMode=BLEND_Masked
  shadingModel=MSM_Unlit
  usedShadingModels=Unlit
 }
 nodes [
  n1 = VectorParameter() {parameterName=GlowstickColor, defaultValue=<rgba 0 1 0.7 1>}
  MaterialFunctionCall(LDR Input=Clamp(None=@n1.RGB), HDR Tint=@n1.RGB, HDR Intensity Multiplier=ScalarParameter() {defaultValue=4, parameterName=HDRIntensityMult}) {materialFunction=&DeriveHDRfromLDR}
  n2 = TextureSample() {samplerType=SAMPLERTYPE_Masks, texture=/BRCosmetics/Athena/Effects/Particles/Skydiving/NS_Contrail_JadeTowel/Textures/T_JadeTowel_Light_FX}
  MaterialFunctionCall(Emissive=LinearInterpolate(A=Multiply(A=@n1.RGB) {constB=0.35}, B=@n1.RGB, Alpha=@n2.G)) {materialFunction=&MF_FX_EmissiveAdjustments}
 ]
}
