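# Translucent unlit, two-sided Niagara sprite material that draws hollow squares for the JadeTowel skydiving contrail.
# Steps on U and V edge-distance products form a procedural border mask, with thickness from dynamic parameter SquareWidth.
# Dynamic parameter UseTextureSample blends particle color toward a radially mapped rainbow gradient.
# The masked color passes through an emissive adjustment function, and the mask is also multiplied by particle alpha.
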Material /BRCosmetics/Athena/Effects/Particles/Skydiving/NS_Contrail_JadeTowel/M_Contrail_JadeTowel_Squares
graph {
 properties {
  bAutomaticallySetUsageInEditor=false
  bUsedWithNiagaraSprites=true
  blendMode=BLEND_Translucent
  shadingModel=MSM_Unlit
  twoSided=true
  usedShadingModels=Unlit
 }
 nodes [
  n1 = ParticleColor()
  n2 = TextureSample(UVs=MaterialFunctionCall(Vector or UVs=TextureCoordinate() {uTiling=0.5, vTiling=0.6}) {materialFunction=&VectorToRadialValue}.Radial Coordinates) {texture=/Game/Effects/Fort_Effects/Textures/Grad/T_Gradient_Rainbow_NoRed}
  n3 = DynamicParameter() {defaultValue=<rgba 0.05 0 1 1>, paramNames [SquareWidth, UseTextureSample, Param3, Param4]}
  n4 = TextureCoordinate()
  n5 = ComponentMask(None=@n4) {r=true}
  n6 = Multiply(A=@n5, B=OneMinus(None=@n5))
  n7 = ComponentMask(None=@n4) {g=true}
  n8 = Multiply(A=OneMinus(None=@n7), B=@n7)
  n9 = Saturate(None=Add(A=Step(Y=@n6, X=@n3.SquareWidth) {constX=0.05}, B=Step(Y=@n8, X=@n3.SquareWidth) {constX=0.05}))
  MaterialFunctionCall(Emissive=Multiply(A=LinearInterpolate(A=@n1.RGB, B=@n2.RGB, Alpha=@n3.UseTextureSample), B=@n9)) {materialFunction=&MF_FX_EmissiveAdjustments}
  Multiply(A=@n9, B=@n1.A)
  ScalarParameter() {defaultValue=0.05, parameterName=SquareWidth}
  ScalarParameter() {parameterName=UseTextureSample?}
  VectorParameter() {defaultValue=<rgba 1 0 0.3176 1>}
 ]
}
