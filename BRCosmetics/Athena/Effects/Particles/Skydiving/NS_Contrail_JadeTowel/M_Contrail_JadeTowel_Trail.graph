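# Translucent unlit, two-sided Niagara sprite material for the trail of the JadeTowel skydiving contrail.
# Panning tile masks, varied by particle random and dynamic-parameter tiling, subtract into a streak mask that multiplies a patterned two-color gradient for emissive.
# Opacity combines that mask with Fresnel, end fades, particle alpha and depth fade.
# The UseAudioBars? scalar blends both outputs toward an audio-bar look: particle color and an AudioBarScale-sized band.
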Material /BRCosmetics/Athena/Effects/Particles/Skydiving/NS_Contrail_JadeTowel/M_Contrail_JadeTowel_Trail
graph {
 properties {
  bCastRayTracedShadows=false
  bEnableResponsiveAA=true
  bUsedWithNiagaraSprites=true
  blendMode=BLEND_Translucent
  pixelDepthOffsetMode=PDOM_Legacy
  shadingModel=MSM_Unlit
  twoSided=true
  usedShadingModels=Unlit
 }
 nodes [
  Constant() {r=33}
  n1 = NamedRerouteUsage()
  LinearInterpolate(Alpha=@n1) {constA=2}
  n2 = ComponentMask(None=TextureCoordinate()) {g=true}
  n3 = ScalarParameter() {defaultValue=0.7, parameterName=Param_2}
  n4 = LinearInterpolate(
   A=VectorParameter() {defaultValue=<rgba 0 1 0.7113 1>}.RGB
   B=VectorParameter() {parameterName=Param_1, defaultValue=<rgba 0.0052 0.8879 0.807 1>}.RGB
   Alpha=MaterialFunctionCall(In=Power(Base=@n2, Exp=@n3), Contrast=Constant() {r=0.1}) {materialFunction=&CheapContrast}.Result
  )
  n5 = TextureCoordinate()
  n6 = AppendVector(A=Constant(), B=ScalarParameter() {defaultValue=0.5, parameterName=Param_9})
  n7 = ScalarParameter() {defaultValue=0.3, parameterName=Param_6}
  n8 = Multiply(A=ScalarParameter() {defaultValue=0.4, parameterName=Param_7}, B=ParticleRandom())
  n9 = TextureSample(UVs=Multiply(A=Panner(Coordinate=@n5, Speed=@n6) {speedY=1}, B=AppendVector(A=@n7, B=@n8))) {texture=/Game/Animation/Game/MainPlayer/Emotes/Veiled/FX/Textures/T_Veiled_Merged_Pattern}
  n10 = ScalarParameter() {defaultValue=0.2, parameterName=Param_5}
  n11 = TextureCoordinate()
  n12 = AppendVector(A=Multiply(A=ParticleRandom()), B=Constant() {r=1})
  n13 = ScalarParameter() {parameterName=SpeedX}
  n14 = ScalarParameter() {defaultValue=1, parameterName=SpeedY}
  n15 = TextureSample(UVs=Panner(Coordinate=Multiply(A=@n11, B=@n12), Time=NamedRerouteUsage(), Speed=AppendVector(A=@n13, B=@n14))) {texture=/Game/Effects/Fort_Effects/Textures/Tile/Tests/T_Tile_Panner_Mask}
  n16 = ScalarParameter() {parameterName=Param_1}
  n17 = ComponentMask(None=TextureCoordinate()) {r=true}
  n18 = Multiply(A=@n17, B=OneMinus(None=@n17))
  n19 = TextureCoordinate()
  n20 = DynamicParameter() {paramNames [Tile_X,Tile_Y,Param3,Param4]}
  n21 = AppendVector(A=@n20.Tile_X, B=@n20.Tile_X)
  n22 = TextureSample(UVs=Panner(Coordinate=Multiply(A=@n19, B=@n21), Time=@n1) {speedY=1}) {texture=&T_Tile_Panner_Mask}
  n23 = ScalarParameter() {parameterName=Param_1}
  n24 = TextureCoordinate()
  n25 = Power(Base=OneMinus(None=ComponentMask(None=@n24) {g=true}), Exp=ScalarParameter() {defaultValue=1.5})
  n26 = ScalarParameter() {defaultValue=11, parameterName=Param_4}
  n27 = Subtract(
   A=LinearInterpolate(
    A=MaterialFunctionCall!(Bit=@n15.RGB, BitMask=@n16) {materialFunction=/Engine/Functions/Engine_MaterialFunctions01/Texturing/BitMask}.Result
    Alpha=SmoothStep(Value=@n18) {constMax=0.15, constMin=0.15}
   )
   B=Multiply(
    A=MaterialFunctionCall(Bit=@n22.RGB, BitMask=@n23) {materialFunction=&BitMask}.Result
    B=MaterialFunctionCall(In=@n25, Contrast=@n26) {materialFunction=&CheapContrast}.Result
   )
  )
  n28 = Multiply(A=Saturate(None=Power(Base=@n9.R, Exp=@n10) {constExponent=0.1}), B=@n27)
  n29 = ParticleColor()
  n30 = ScalarParameter() {parameterName=UseAudioBars?}
  MaterialFunctionCall(Emissive=LinearInterpolate(A=Multiply(A=@n4, B=@n28), B=@n29.RGB, Alpha=@n30)) {materialFunction=&MF_FX_EmissiveAdjustments}
  n31 = Power(Base=ComponentMask(None=TextureCoordinate()) {g=true}, Exp=Constant() {r=1})
  n32 = OneMinus(None=ComponentMask(None=TextureCoordinate()) {g=true})
  n33 = Multiply(A=Multiply(A=@n27, B=Fresnel() {exponent=1}), B=Multiply(A=@n31, B=Power(Base=@n32) {constExponent=0.5}))
  n34 = Multiply(A=Constant() {r=0.05}, B=ScalarParameter() {parameterName=AudioBarScale})
  n35 = OneMinus(None=MaterialFunctionCall!(Width=Clamp(None=@n34) {maxDefault=0.5}, Sharpness, Offset=Constant() {r=0.4}, Direction Switch, Compare, Input Coordinates) {materialFunction=/Engine/Functions/Engine_MaterialFunctions03/Procedurals/GeneratedBand}.Result)
  Multiply(
   A=LinearInterpolate(A=Multiply(A=@n29.A, B=@n33), B=Multiply(A=@n29.A, B=@n35), Alpha=@n30)
   B=DepthFade(FadeDistance=ScalarParameter() {defaultValue=22, parameterName=Param_3})
  )
  Multiply(A=@n31, B=SmoothStep(Value=@n31) {constMax=0.2})
  NamedRerouteDeclaration(None=ScalarParameter() {parameterName=UVPanTime}) {name=UV Pan Time, nodeColor=<rgba 0 1 0.9063 1>}
  ScalarParameter() {defaultValue=0.7, parameterName=Param_2}
  SmoothStep(Value=@n31) {constMax=0.5, constMin=-0.1}
  TextureSample() {texture=&T_Tile_Panner_Mask}
 ]
}
