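# Unlit, two-sided translucent Niagara sprite material that draws a scrolling circuit-pattern trail for the JadeTowel skydiving contrail.
# The VitalPsych circuit texture, randomized per particle by sub-UV, is multiplied by cell noise and particle color for emissive adjustment.
# An alpha-like chain blends in HDR-boosted moving lights and subtracts procedural breakup, driven by the CircuitLights and CircuitBreakUp dynamic parameters.
# That chain is scaled by particle alpha and faded along the V coordinate.
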Material /BRCosmetics/Athena/Effects/Particles/Skydiving/NS_Contrail_JadeTowel/M_Contrail_JadeTowel_Trail_Circuit
graph {
 properties {
  bAutomaticallySetUsageInEditor=false
  bUsedWithNiagaraSprites=true
  blendMode=BLEND_Translucent
  shadingModel=MSM_Unlit
  twoSided=true
  usedShadingModels=Unlit
 }
 nodes [
  AppendVector(A=ScalarParameter() {parameterName=Param_5}, B=ScalarParameter() {defaultValue=0.5, parameterName=Param_6})
  Constant() {r=4}
  n1 = ParticleColor()
  n2 = ScalarParameter() {defaultValue=1, parameterName=Param_7}
  n3 = ScalarParameter() {defaultValue=5, parameterName=Param_8}
  n4 = ParticleRandom()
  n5 = TextureSample(UVs=Panner(
    Coordinate=MaterialFunctionCall!(UseTexCoord 1, Rows=@n2, Columns=@n3, Particle Random Value=@n4) {materialFunction=/Game/Effects/Fort_Effects/Materials/_Functions/MF_RandomSubUV_GPU_Sprites}.Result
    Time=NamedRerouteUsage()
   ) {speedY=0.5}) {texture=/BRCosmetics/Athena/Effects/Particles/Skydiving/NS_Contrail_VitalPsych/Textures/T_VitalPsych_Circuit}
  n6 = Panner(Coordinate=TextureCoordinate() {uTiling=0.2, vTiling=0.2}) {speedY=0.5}
  n7 = Saturate(None=Multiply(A=TextureSample(UVs=@n6) {samplerType=SAMPLERTYPE_LinearGrayscale, texture=/Game/Effects/Fort_Effects/Textures/Tile/T_CellNoise_Tile_M}.RGB) {constB=5})
  MaterialFunctionCall(Emissive=Multiply(A=@n1.RGB, B=Multiply(A=@n5.RGB, B=@n7))) {materialFunction=&MF_FX_EmissiveAdjustments}
  n8 = Multiply(A=@n5.RGB) {constB=0.2}
  n9 = MaterialFunctionCall(UVs=Panner(Coordinate=TextureCoordinate(), Time=NamedRerouteUsage()) {speedY=0.75}, "Rotation Angle (0-1)"=Constant() {r=0.75}) {materialFunction=&CustomRotator}
  n10 = Multiply(
   A=@n5.RGB
   B=Power(Base=TextureSample(UVs=@n9.Rotated Values) {texture=/Game/Effects/Fort_Effects/Textures/Tile/Tests/T_Tile_Panner_Mask}.RGB)
  )
  n11 = MaterialFunctionCall(LDR Input=Clamp(None=@n10), HDR Tint=@n10, HDR Intensity Multiplier=ScalarParameter() {defaultValue=4, parameterName=HDRIntensityMult}) {materialFunction=&DeriveHDRfromLDR}
  n12 = Multiply(
   A=LinearInterpolate(
    A=@n5.RGB
    B=Add(A=@n8, B=@n11.RemappedRange)
    Alpha=DynamicParameter() {defaultValue=<rgba 0 0 1 1>, paramNames [CircuitLights, CircuitBreakUp, Param3, Param4]}.CircuitLights
   )
   B=ComponentMask(None=TextureCoordinate()) {g=true}
  )
  n13 = TextureCoordinate() {uTiling=0.5, vTiling=0.5}
  n14 = AppendVector(A=ScalarParameter() {parameterName=SpeedX}, B=ScalarParameter() {defaultValue=0.5, parameterName=SpeedY})
  n15 = MaterialFunctionCall!(
   Bit=TextureSample(UVs=Panner(Coordinate=@n13, Speed=@n14)) {texture=&T_Tile_Panner_Mask}.RGB
   BitMask=ScalarParameter() {parameterName=Param_1}
  ) {materialFunction=/Engine/Functions/Engine_MaterialFunctions01/Texturing/BitMask}
  n16 = TextureCoordinate()
  n17 = AppendVector(A=ScalarParameter() {defaultValue=1, parameterName=Param_3}, B=ScalarParameter() {defaultValue=22, parameterName=Param_4})
  n18 = ScalarParameter() {parameterName=SpeedX}
  n19 = ScalarParameter() {defaultValue=0.5, parameterName=SpeedY}
  n20 = MaterialFunctionCall!(
   Width=Constant() {r=0.1}
   Sharpness=ScalarParameter() {defaultValue=1}
   Offset
   Direction Switch
   Compare
   Input Coordinates=Panner(Coordinate=Multiply(A=@n16, B=@n17), Speed=AppendVector(A=@n18, B=@n19))
  ) {materialFunction=/Engine/Functions/Engine_MaterialFunctions03/Procedurals/GeneratedBand}
  n21 = LinearInterpolate(
   A=Constant()
   B=Saturate(None=Subtract(A=@n15.Result, B=@n20.Result))
   Alpha=DynamicParameter() {defaultValue=<rgba 0 0 1 1>, paramNames [CircuitLights, CircuitBreakUp, Param3, Param4]}.CircuitBreakUp
  )
  n22 = ComponentMask(None=TextureCoordinate()) {g=true}
  n23 = ScalarParameter() {defaultValue=2, parameterName=Param_2}
  Multiply(A=Multiply(A=@n1.A, B=Subtract(A=@n12, B=@n21)), B=SmoothStep(Value=Power(Base=@n22, Exp=@n23)) {constMax=0.5, constMin=-0.05})
  NamedRerouteDeclaration(None=ScalarParameter() {parameterName=UVPanTime}) {name=UVPanTime, nodeColor=<rgba 0 0.9049 1 1>}
 ]
}
