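# NS_Contrail_JadeTowel is a skydiving contrail cosmetic themed as an audio-reactive light show.
# Emitters attach glowsticks, lights, spinning squares, body glow, trails and CPU/GPU sparkles to skeletal mesh bones, alongside two swaying stage lights.
# FN_Contrail_Setup supplies dive value and direction for intensity, stretch and orientation.
# System audio scales modulate waveform-driven size, color and brightness, while body glow spawns only for local player or front-end preview.
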NiagaraSystem /BRCosmetics/Athena/Effects/Particles/Skydiving/NS_Contrail_JadeTowel/NS_Contrail_JadeTowel
dependencies {
 usedDataInterfaces [
  &NiagaraDataInterfaceArrayColor
  &NiagaraDataInterfaceCurve
  &NiagaraDataInterfaceArrayFloat
  &NiagaraDataInterfaceArrayInt32
  &NiagaraDataInterfaceArrayFloat3
  &NiagaraDataInterfaceColorCurve
  &NiagaraDataInterfaceSkeletalMesh
 ]
 usedDynamicInputs [&RemapRange, &MakeFloatFromLinearColor]
 usedModules [
  &self:SystemSpawnScript.SystemScriptSource.SystemScriptGraph.NiagaraNodeAssignment_3.SetVariables_#1
  &SystemState
  &Timeline
  &self:SystemSpawnScript.SystemScriptSource.SystemScriptGraph.NiagaraNodeAssignment_0.SetVariables_#2
  &self:SystemSpawnScript.SystemScriptSource.SystemScriptGraph.NiagaraNodeAssignment_6.SetVariables_#3
  /Game/Effects/Niagara/Modules/Contrails/FN_Contrail_Setup
  &FN_Integrator
  &self:SystemSpawnScript.SystemScriptSource.SystemScriptGraph.NiagaraNodeAssignment_4.SetVariables_#4
 ]
}
emitters {
 Glow_Body001 {
  emitterUpdateScript [
   /SetVariables(Emitter.GlowOutlineThickness:NiagaraFloat=$System.GlowOutsideOutline) {isSetParameters=true}
   &EmitterState(Life Cycle Mode=::NewEnumerator1, Loop Behavior=::NewEnumerator1, Loop Duration Mode=::NewEnumerator1, UseLoopDelay=false)
   &SpawnBurst_Instantaneous(
    Spawn Count=&MakeCustomIntFromBool(
     Bool=$User.bIsFrontEndPreview
     True INT=1
     False INT=&MakeCustomIntFromBool(Bool=$User.bIsLocalPlayer, True INT=1, False INT=0)
    )
    Use Spawn Probability=false
    UseLoopCountLimit=false
   )
  ]
  particleSpawnScript [&InitializeParticle(Mesh Scale=<2.5 1 1>, Mesh Scale Mode=::NewEnumerator4, UsePositionOffset=false, Write Material Random=true)]
  particleUpdateScript [
   &ParticleState(
    Kill Particles When Lifetime Has Elapsed=false
    Let Infinitely Lived Particles Die When Emitter Deactivates=true
    Loop Particles Lifetime=true
   )
   &SkeletalMeshLocation(Bone Sampling Mode=::NewEnumerator4, MaskSpawnGroup=false, Orientation Sampling=::NewEnumerator1, Skeletal Mesh=$System.Skeletal Mesh) {module=SkeletalMeshLocation001}
   &CameraOffset(Camera Offset Amount=&Multiply_Float(A=22, B=$System.FN_Contrail_Setup.Dive Value))
   &ScaleColor(Scale Alpha=&Lerp_Float(A=44, B=90, Alpha=$System.FN_Contrail_Setup.Dive Value), ScaleA=true, ScaleRGB=true) {module=ScaleColor001}
  ]
  rendererClasses [&NiagaraMeshRendererProperties]
  renderers [
   {rendererClass=&NiagaraMeshRendererProperties, rendererIndex=0}
   {rendererClass=&NiagaraMeshRendererProperties, rendererIndex=1}
  ]
 }
 Glowsticks001 {
  emitterUpdateScript [
   /SetVariables(Emitter.GlowstickColor:LinearColor=$System.GlowstickColor) {isSetParameters=true}
   &EmitterState(Life Cycle Mode=::NewEnumerator1, Loop Behavior=::NewEnumerator1, Loop Duration Mode=::NewEnumerator1, UseLoopDelay=false)
   &SpawnBurst_Instantaneous(Spawn Count=2, Use Spawn Probability=false, UseLoopCountLimit=false)
  ]
  particleSpawnScript [
   &PartitionParticles()
   &InitializeParticle(
    Color=<rgba 0.0052 0.8879 0.807 1>
    Mesh Scale Mode=::NewEnumerator3
    Mesh Uniform Scale=0.7
    UsePositionOffset=false
    Write Material Random=true
   )
   &InitialMeshOrientation(
    Mesh Orientation Mode=::NewEnumerator7
    Rotation=&Multiply_VectorByFloat(
     Vector=<0 0.05 0>
     Float=&Lerp_Float(A=1, B=-1, Alpha=&MakeFloatFromInt(Int=$Particles.PartitionParticles.Partition))
    )
    Use Rotation Vector=true
   )
  ]
  particleUpdateScript [
   &ParticleState(
    Kill Particles When Lifetime Has Elapsed=false
    Let Infinitely Lived Particles Die When Emitter Deactivates=true
    Loop Particles Lifetime=true
   )
   &UpdateMeshOrientation(
    Rotation Rate=&Sine(Normalized Angle=$Emitter.Age, Period=1, Scale=&RandomRangeFloat(Minimum=0.4, Maximum=0.6), Bias=0)
    Rotation Vector=<-0.5 0 0>
   )
   &UpdateMeshOrientation(
    Rotation Rate=&Sine(
     Normalized Angle=$Emitter.Age
     Period=1
     Scale=&Multiply_Float(
      A=&Lerp_Float(
       A=&RandomRangeFloat(Minimum=-0.1, Maximum=-0.1)
       B=&RandomRangeFloat(Minimum=-0.3, Maximum=-0.4)
       Alpha=$System.FN_Contrail_Setup.Dive Value
      )
      B=&Lerp_Float(A=1.5, B=-1.5, Alpha=&MakeFloatFromInt(Int=$Particles.PartitionParticles.Partition))
     )
     Bias=0
    )
    Rotation Vector=<0 1 0>
   ) {module=UpdateMeshOrientation001}
   &SkeletalMeshLocation(
    Bone / Socket Index=&SelectIntFromArray(Direct Array Index=$Particles.PartitionParticles.Partition, Int Selection Array=DataInterface() {propertyValues {IntData [6,7]}})
    Bone Sampling Mode=::NewEnumerator4
    MaskSpawnGroup=false
    Orientation Sampling=::NewEnumerator1
    Sampled Position Offset=&Multiply_VectorByFloat(
     Vector=<18 -5 0>
     Float=&Lerp_Float(A=1, B=-1, Alpha=&MakeFloatFromInt(Int=$Particles.PartitionParticles.Partition))
    )
    Skeletal Mesh=$System.Skeletal Mesh
   )
  ]
  rendererClasses [&NiagaraMeshRendererProperties]
  renderers [
   {rendererClass=&NiagaraMeshRendererProperties, rendererIndex=0}
   {rendererClass=&NiagaraMeshRendererProperties, rendererIndex=1}
  ]
 }
 Light001 {
  emitterUpdateScript [
   &EmitterState(Life Cycle Mode=::NewEnumerator1, Loop Behavior=::NewEnumerator1, Loop Duration Mode=::NewEnumerator1, UseLoopDelay=false)
   &SpawnBurst_Instantaneous(Spawn Count=4, Use Spawn Probability=false, UseLoopCountLimit=false)
  ]
  particleSpawnScript [
   &PartitionParticles(Number Of Partitions=4)
   &InitializeParticle(
    Color=<rgba 0.0052 0.8879 0.807 1>
    Sprite Size Mode=::NewEnumerator3
    Uniform Sprite Size=&SelectFloatFromArray(Direct Array Index=$Particles.PartitionParticles.Partition, Float Selection Array=DataInterface() {propertyValues {FloatData [50,50,30,30]}})
    UsePositionOffset=false
    Write Material Random=true
   )
  ]
  particleUpdateScript [
   &ParticleState(
    Kill Particles When Lifetime Has Elapsed=false
    Let Infinitely Lived Particles Die When Emitter Deactivates=true
    Loop Particles Lifetime=true
   )
   &SkeletalMeshLocation(
    Bone / Socket Index=&SelectIntFromArray(Direct Array Index=$Particles.PartitionParticles.Partition, Int Selection Array=DataInterface() {propertyValues {IntData [6,7,4,5]}})
    Bone Sampling Mode=::NewEnumerator4
    MaskSpawnGroup=false
    Orientation Sampling=::NewEnumerator1
    Sampled Position Offset=&SelectVectorFromArray(
     Direct Array Index=$Particles.PartitionParticles.Partition
     Vector Selection Array=DataInterface() {propertyValues {FloatData [<5 -5 0>, <-5 5 0>, <-9 0 0>, <9 0 0>]}}
    )
    Skeletal Mesh=$System.Skeletal Mesh
   )
   &FN_Light_Attributes(Color=<rgba 0.1284 22 19.9938 1>, Create Color Attribute=true, Enable Particle Light Enable=true, Light Enabled=true)
   &CameraOffset(Camera Offset Amount=9)
   &ScaleColor(
    Scale Alpha=&FloatFromCurve(
     FloatCurve=DataInterface() {propertyValues {Curve {keys [
         {interpMode=RCIM_Cubic, tangentMode=RCTM_User, time=0, value=0}
         {arriveTangent=2.5708, interpMode=RCIM_Cubic, leaveTangent=2.5708, tangentMode=RCTM_User, time=1, value=1}
        ]}}}
     CurveIndex=$Emitter.Age
     Scale Curve=0.4
    )
    ScaleA=true
    ScaleRGB=false
   )
  ]
  rendererClasses [&NiagaraLightRendererProperties, &NiagaraSpriteRendererProperties]
  renderers [
   {rendererClass=&NiagaraLightRendererProperties, rendererIndex=0}
   {rendererClass=&NiagaraSpriteRendererProperties, rendererIndex=1}
  ]
 }
 Sparkles {
  emitterUpdateScript [
   &EmitterState()
   &SpawnRate(SpawnRate=&Lerp_Float(A=5, B=27, Alpha=$System.FN_Contrail_Setup.Dive Value), Use Spawn Probability=false) {module=SpawnRate001}
  ]
  particleSpawnScript [
   &PartitionParticles()
   &InitializeParticle(
    Lifetime=&Lerp_Float(
     A=&RandomRangeFloat(Minimum=0.4, Maximum=0.9)
     B=&RandomRangeFloat(Minimum=0.15, Maximum=0.6)
     Alpha=$System.FN_Contrail_Setup.Dive Value
    )
    Sprite Rotation Mode=::NewEnumerator2
    Sprite Size Mode=::NewEnumerator1
    Uniform Sprite Size Min=17
    UsePositionOffset=false
    Write Material Random=true
   )
   &AddVelocity(
    Cone Angle=30
    Cone Axis=&Lerp_Vector(
     A=&FN_FrontEndVector3Modifier(Gameplay=<0 0 1>, Front End=<0 -0.3 1>)
     B=&FN_FrontEndVector3Modifier(Gameplay=<0 0 1>, Front End=<0 -0.9 1>)
     Alpha=$System.FN_Contrail_Setup.Dive Value
    )
    Module.UseConeFalloff=false
    Module.UseVelDistribution=false
    Override Seed=false
    Velocity Mode=::NewEnumerator2
    Velocity Speed=&Lerp_Float(A=300, B=&RandomRangeFloat(Minimum=500, Maximum=700), Alpha=$System.FN_Contrail_Setup.Dive Value)
   )
   &SkeletalMeshLocation(
    Bone / Socket Index=$Particles.PartitionParticles.Partition
    Bone Sampling Mode=::NewEnumerator4
    MaskSpawnGroup=false
    Skeletal Mesh=$System.Skeletal Mesh
   ) {module=SkeletalMeshLocation001}
  ]
  particleUpdateScript [
   &ParticleState()
   /SetVariables(Particles.AudioWaveform:NiagaraFloat=&Waveform(
     Global Phase=&Add_Float(A=$Emitter.Age, B=$System.AudioScale)
     "Waveform [1]"=::NewEnumerator0
     "[1] Frequency"=1
     "[1] Amplitude Min/Max"=<0.1 1>
     "[1] Scale"=0.4
    )) {isSetParameters=true, module=SetVariables_#5}
   /SetVariables(
    Particles.GlowColor=&MultiplyLinearColor(
     Linear Color=&SelectLinearColorFromArray(
      Array Sampling Mode=::NewEnumerator0
      Direct Array Index=?
      Color Selection Array=DataInterface() {propertyValues {ColorData [<rgba 3 0.0961 2.9812 1>, <rgba 2 1.6172 0.0553 1>, <rgba 0.0961 2.0591 3 1>, <rgba 0.7086 0.07 3 1>, <rgba 0.1148 2 0.1075 1>]}}
      Random Seed=0
     )
     Scale Factor=&MakeVector4(
      X=1
      Y=1
      Z=1
      W=&Multiply_Float(
       A=&Lerp_Float(A=0.1, B=$Particles.AudioWaveform, Alpha=$System.AudioScale)
       B=&FloatFromCurve(FloatCurve=DataInterface() {propertyValues {Curve {keys [
            {interpMode=RCIM_Cubic, leaveTangent=-2.4988, tangentMode=RCTM_Break, time=0.5, value=1}
            {interpMode=RCIM_Cubic, tangentMode=RCTM_User, time=1, value=0}
           ]}}})
      )
     )
    )
    Particles.GlowScale:Vector2f=&Vector2DFromFloat(Value=&Multiply_Float(
      A=&RandomRangeFloat(Minimum=35, Maximum=25)
      B=&FloatFromCurve(FloatCurve=DataInterface() {propertyValues {Curve {keys [
           {interpMode=RCIM_Cubic, tangentMode=RCTM_User, time=0, value=0}
           {interpMode=RCIM_Cubic, tangentMode=RCTM_User, time=0.2, value=1}
           {arriveTangent=0.2312, interpMode=RCIM_Cubic, leaveTangent=-5.6641, tangentMode=RCTM_Break, time=0.7, value=1}
           {arriveTangent=-0.2123, interpMode=RCIM_Cubic, tangentMode=RCTM_Break, time=1, value=0}
          ]}}})
     ))
   ) {isSetParameters=true, module=SetVariables_#6}
   &Drag(Drag=2, Use Linear Drag=true, Use Rotational Drag=true) {module=Drag001}
   &GravityForce(Gravity=&RandomRangeVector(Minimum=<0 0 200>, Maximum=<1 1 300>))
   &ScaleSpriteSize(
    Scale Factor=&Vector2DFromFloat(Value=&FloatFromCurve(FloatCurve=DataInterface() {propertyValues {Curve {keys [<at 0.0018 0.1957>,<at 0.0527 0.819>,<at 0.1169 0.6555>,<at 0.1748 0.9015>,<at 0.2394 0.582>,<at 0.332 1.0023>,<at 0.4947 0.3707>,<at 0.6727 1>,<at 1 0>]}}}))
    Scale Sprite Size Mode=::NewEnumerator2
   )
   &ScaleSpriteSize(Uniform Curve Sprite Scale=DataInterface() {propertyValues {Curve {keys [
        {interpMode=RCIM_Cubic, tangentMode=RCTM_User, time=0, value=0}
        {interpMode=RCIM_Cubic, tangentMode=RCTM_User, time=0.2, value=1}
        {arriveTangent=0.2312, interpMode=RCIM_Cubic, leaveTangent=-5.6641, tangentMode=RCTM_Break, time=0.7, value=1}
        {arriveTangent=-0.2123, interpMode=RCIM_Cubic, tangentMode=RCTM_Break, time=1, value=0}
       ]}}}) {module=ScaleSpriteSize001}
   &SolveForcesAndVelocity(Clamp Velocity=false, Limit Acceleration=false)
  ]
  rendererClasses [&NiagaraSpriteRendererProperties]
  renderers [
   {rendererClass=&NiagaraSpriteRendererProperties}
   {rendererClass=&NiagaraSpriteRendererProperties, rendererIndex=1}
  ]
 }
 Sparkles_GPU {
  simTarget=GPUComputeSim
  emitterUpdateScript [
   &EmitterState()
   &SpawnRate(SpawnRate=&Lerp_Float(A=5, B=27, Alpha=$System.FN_Contrail_Setup.Dive Value), Use Spawn Probability=false) {module=SpawnRate001}
  ]
  particleSpawnScript [
   &PartitionParticles()
   &InitializeParticle(
    Lifetime=&Lerp_Float(
     A=&RandomRangeFloat(Minimum=0.4, Maximum=0.9)
     B=&RandomRangeFloat(Minimum=0.15, Maximum=0.6)
     Alpha=$System.FN_Contrail_Setup.Dive Value
    )
    Sprite Rotation Mode=::NewEnumerator2
    Sprite Size Mode=::NewEnumerator1
    Uniform Sprite Size Min=17
    UsePositionOffset=false
    Write Material Random=true
   )
   &AddVelocity(
    Cone Angle=30
    Cone Axis=&Lerp_Vector(
     A=&FN_FrontEndVector3Modifier(Gameplay=<0 0 1>, Front End=<0 -0.3 1>)
     B=&FN_FrontEndVector3Modifier(Gameplay=<0 0 1>, Front End=<0 -0.9 1>)
     Alpha=$System.FN_Contrail_Setup.Dive Value
    )
    Module.UseConeFalloff=false
    Module.UseVelDistribution=false
    Override Seed=false
    Velocity Mode=::NewEnumerator2
    Velocity Speed=&Lerp_Float(A=300, B=&RandomRangeFloat(Minimum=500, Maximum=700), Alpha=$System.FN_Contrail_Setup.Dive Value)
   )
   &SkeletalMeshLocation(
    Bone / Socket Index=$Particles.PartitionParticles.Partition
    Bone Sampling Mode=::NewEnumerator4
    MaskSpawnGroup=false
    Skeletal Mesh=$System.Skeletal Mesh
   ) {module=SkeletalMeshLocation001}
  ]
  particleUpdateScript [
   &ParticleState()
   /SetVariables(Particles.AudioWaveform:NiagaraFloat=&Waveform(
     Global Phase=&Add_Float(A=$Emitter.Age, B=$System.AudioScale)
     "Waveform [1]"=::NewEnumerator0
     "[1] Frequency"=1
     "[1] Amplitude Min/Max"=<0.1 1>
     "[1] Scale"=0.4
    )) {isSetParameters=true, module=SetVariables_#5}
   /SetVariables(
    Particles.GlowColor=&MultiplyLinearColor(
     Linear Color=&SelectLinearColorFromArray(
      Array Sampling Mode=::NewEnumerator0
      Direct Array Index=?
      Color Selection Array=DataInterface() {propertyValues {ColorData [<rgba 3 0.0961 2.9812 1>, <rgba 2 1.6172 0.0553 1>, <rgba 0.0961 2.0591 3 1>, <rgba 0.7086 0.07 3 1>, <rgba 0.1148 2 0.1075 1>]}}
      Random Seed=0
     )
     Scale Factor=&MakeVector4(
      X=1
      Y=1
      Z=1
      W=&Multiply_Float(
       A=&Lerp_Float(A=0.1, B=$Particles.AudioWaveform, Alpha=$System.AudioScale)
       B=&FloatFromCurve(FloatCurve=DataInterface() {propertyValues {Curve {keys [
            {interpMode=RCIM_Cubic, leaveTangent=-2.4988, tangentMode=RCTM_Break, time=0.5, value=1}
            {interpMode=RCIM_Cubic, tangentMode=RCTM_User, time=1, value=0}
           ]}}})
      )
     )
    )
    Particles.GlowScale:Vector2f=&Vector2DFromFloat(Value=&Multiply_Float(
      A=&RandomRangeFloat(Minimum=35, Maximum=25)
      B=&FloatFromCurve(FloatCurve=DataInterface() {propertyValues {Curve {keys [
           {interpMode=RCIM_Cubic, tangentMode=RCTM_User, time=0, value=0}
           {interpMode=RCIM_Cubic, tangentMode=RCTM_User, time=0.2, value=1}
           {arriveTangent=0.2312, interpMode=RCIM_Cubic, leaveTangent=-5.6641, tangentMode=RCTM_Break, time=0.7, value=1}
           {arriveTangent=-0.2123, interpMode=RCIM_Cubic, tangentMode=RCTM_Break, time=1, value=0}
          ]}}})
     ))
   ) {isSetParameters=true, module=SetVariables_#6}
   &Drag(Drag=2, Use Linear Drag=true, Use Rotational Drag=true) {module=Drag001}
   &GravityForce(Gravity=&RandomRangeVector(Minimum=<0 0 200>, Maximum=<1 1 300>))
   &ScaleSpriteSize(
    Scale Factor=&Vector2DFromFloat(Value=&FloatFromCurve(FloatCurve=DataInterface() {propertyValues {Curve {keys [<at 0.0018 0.1957>,<at 0.0527 0.819>,<at 0.1169 0.6555>,<at 0.1748 0.9015>,<at 0.2394 0.582>,<at 0.332 1.0023>,<at 0.4947 0.3707>,<at 0.6727 1>,<at 1 0>]}}}))
    Scale Sprite Size Mode=::NewEnumerator2
   )
   &ScaleSpriteSize(Uniform Curve Sprite Scale=DataInterface() {propertyValues {Curve {keys [
        {interpMode=RCIM_Cubic, tangentMode=RCTM_User, time=0, value=0}
        {interpMode=RCIM_Cubic, tangentMode=RCTM_User, time=0.2, value=1}
        {arriveTangent=0.2312, interpMode=RCIM_Cubic, leaveTangent=-5.6641, tangentMode=RCTM_Break, time=0.7, value=1}
        {arriveTangent=-0.2123, interpMode=RCIM_Cubic, tangentMode=RCTM_Break, time=1, value=0}
       ]}}}) {module=ScaleSpriteSize001}
   &SolveForcesAndVelocity(Clamp Velocity=false, Limit Acceleration=false)
  ]
  rendererClasses [&NiagaraSpriteRendererProperties]
  renderers [
   {rendererClass=&NiagaraSpriteRendererProperties}
   {rendererClass=&NiagaraSpriteRendererProperties, rendererIndex=1}
  ]
 }
 StageLights {
  emitterUpdateScript [
   &EmitterState(Life Cycle Mode=::NewEnumerator1, Loop Behavior=::NewEnumerator1, Loop Duration Mode=::NewEnumerator1, UseLoopDelay=false)
   &SpawnBurst_Instantaneous(Spawn Count=2, Use Spawn Probability=false, UseLoopCountLimit=false)
  ]
  particleSpawnScript [
   &PartitionParticles()
   &InitializeParticle(
    Color Mode=::NewEnumerator0
    Mesh Scale=<0.3 0.3 1>
    Mesh Scale Mode=::NewEnumerator4
    Sprite Size Mode=::NewEnumerator3
    UsePositionOffset=false
    Write Material Random=true
   )
   /SetVariables(Particles.Position=&AddVectorToPosition(
     Position=$Particles.Position
     Vector=&MakeVector(
      X=&Multiply_Float(
       A=150
       B=&Lerp_Float(A=-1, B=1, Alpha=&MakeFloatFromInt(Int=$Particles.PartitionParticles.Partition))
      )
      Y=99
      Z=-20
     )
    )) {isSetParameters=true}
   &InitialMeshOrientation(
    Mesh Axis To Orient=<0 0 1>
    Mesh Orientation Mode=::NewEnumerator2
    Orientation Vector=<0 0 -1>
    Rotation=&Multiply_VectorByFloat(
     Vector=<0 -0.1 0>
     Float=&Lerp_Float(A=-1, B=1, Alpha=&MakeFloatFromInt(Int=$Particles.PartitionParticles.Partition))
    )
    Use Rotation Vector=true
   )
  ]
  particleUpdateScript [
   &ParticleState(
    Kill Particles When Lifetime Has Elapsed=false
    Let Infinitely Lived Particles Die When Emitter Deactivates=true
    Loop Particles Lifetime=true
   )
   &UpdateMeshOrientation(
    Rotation Rate=&Multiply_Float(
     A=1
     B=&Waveform("Waveform [1]"=::NewEnumerator0, "[1] Phase Scale"=0.3, "[1] Frequency"=1, "[1] Amplitude Min/Max"=<-1 1>)
    )
    Rotation Vector=&MakeVector(
     X=0.15
     Y=&Multiply_Float(
      A=0.2
      B=&Lerp_Float(A=-1, B=1, Alpha=&MakeFloatFromInt(Int=$Particles.PartitionParticles.Partition))
     )
     Z=0
    )
   )
   &RotateAroundPoint(
    Radius=120
    Rotation Rate:NiagaraFloat=&Waveform(
     "Waveform [1]"=::NewEnumerator0
     "[1] Phase Scale"=&RandomRangeFloat(Minimum=0.1, Maximum=0.2)
     "[1] Frequency"=1
     "[1] Amplitude Min/Max"=<-1 1>
     "[1] Scale"=0.07
    )
    X Axis:Vector3f=&Multiply_VectorByFloat(
     Vector=<1 0 0>
     Float=&Lerp_Float(A=-1, B=1, Alpha=&MakeFloatFromInt(Int=$Particles.PartitionParticles.Partition))
    )
   ) {module=RotateAroundPoint001}
   /SetVariables(Particles.Position=&AddVectorToPosition(
     Position=$StackContext.Position
     Vector=&MakeVector(
      X=0
      Y=0
      Z=&Waveform(
       "Waveform [1]"=::NewEnumerator0
       "[1] Frequency"=&RandomRangeFloat(Minimum=0.5, Maximum=1)
       "[1] Amplitude Min/Max"=<-1 1>
       "[1] Scale"=&RandomRangeFloat(Minimum=4, Maximum=6)
      )
     )
    )) {isSetParameters=true}
   &ScaleColor(
    Scale Alpha=&Multiply_Float(
     A=&Lerp_Float(
      A=0.4
      B=&Waveform(
       Global Phase=&Add_Float(A=$Emitter.Age, B=$System.AudioScale)
       "Waveform [1]"=::NewEnumerator0
       "[1] Frequency"=1
       "[1] Amplitude Min/Max"=<0.5 3>
      )
      Alpha=$System.AudioScale
     )
     B=&FloatFromCurve(
      FloatCurve=DataInterface() {propertyValues {Curve {keys [
          {interpMode=RCIM_Cubic, tangentMode=RCTM_User, time=0, value=0}
          {arriveTangent=2.5708, interpMode=RCIM_Cubic, leaveTangent=2.5708, tangentMode=RCTM_User, time=1, value=1}
         ]}}}
      CurveIndex=$Emitter.Age
      Scale Curve=0.4
     )
    )
    Scale RGB=&Multiply_Vector(
     A=&MakeVectorFromLinearColorRGB(Color=&ColorFromCurve(
       ColorCurve=DataInterface() {propertyValues {
         AlphaCurve {keys [<at 0 1 RCIM_Cubic>]}
         BlueCurve {keys [<at 0 0.807 RCIM_Cubic>,<at 0.2246 0>,<at 0.3968 0.019>,<at 0.6058 0.8879>,<at 0.8235 0.8879>,<at 1 0.807 RCIM_Cubic>]}
         GreenCurve {keys [<at 0 0.8879 RCIM_Cubic>,<at 0.2246 1>,<at 0.3968 1>,<at 0.6058 0.3043>,<at 0.8235 0>,<at 1 0.8879 RCIM_Cubic>]}
         RedCurve {keys [<at 0 0.0052 RCIM_Cubic>,<at 0.2246 0.9451>,<at 0.3968 0.1838>,<at 0.6058 0>,<at 0.8235 0.4788>,<at 1 0.0052 RCIM_Cubic>]}
        }}
       CurveIndex=&Multiply_Float(A=&Sine(Normalized Angle=$Emitter.Age, Period=2, Scale=1, Bias=0), B=$System.AudioScale)
      ))
     B=&VectorFromFloat(Value=1)
    )
    ScaleA=true
    ScaleRGB=true
   ) {module=ScaleColor001}
   &DynamicMaterialParameters(
    Index 0 Param 1=0.15
    Module.Index0Param0Write=true
    Module.Index0Param1Write=true
    Module.Index0Param2Write=true
    Module.Index0Param3Write=true
   )
   /SetVariables(
    Particles.RingPosition:NiagaraPosition=&AddVectorToPosition(Position=$Particles.Position, Vector=<0 0 3>)
    Particles.RingScale:Vector3f=&VectorFromFloat(Value=0.25)
   ) {isSetParameters=true}
   /SetVariables(Particles.DiskScale:Vector3f=&Multiply_VectorByFloat(Vector=<1 1 0.5>, Float=1)) {isSetParameters=true}
   /SetVariables(
    Particles.GlowColor=&MakeLinearColorFromVectorAndFloat(
     "Vector (RGB)"=&Multiply_Vector(
      A=&MakeVectorFromLinearColorRGB(Color=&ColorFromCurve(
        ColorCurve=DataInterface() {propertyValues {
          AlphaCurve {keys [<at 0 1 RCIM_Cubic>]}
          BlueCurve {keys [<at 0 0.807 RCIM_Cubic>,<at 0.2246 0>,<at 0.3968 0.019>,<at 0.6058 0.8879>,<at 0.8235 0.8879>,<at 1 0.807 RCIM_Cubic>]}
          GreenCurve {keys [<at 0 0.8879 RCIM_Cubic>,<at 0.2246 1>,<at 0.3968 1>,<at 0.6058 0.3043>,<at 0.8235 0>,<at 1 0.8879 RCIM_Cubic>]}
          RedCurve {keys [<at 0 0.0052 RCIM_Cubic>,<at 0.2246 0.9451>,<at 0.3968 0.1838>,<at 0.6058 0>,<at 0.8235 0.4788>,<at 1 0.0052 RCIM_Cubic>]}
         }}
        CurveIndex=&Multiply_Float(A=&Sine(Normalized Angle=$Emitter.Age, Period=2, Scale=1, Bias=0), B=$System.AudioScale)
       ))
      B=&VectorFromFloat(Value=1)
     )
     "Float (Alpha)"=0.3
    )
    Particles.GlowScale:Vector2f=&Vector2DFromFloat(Value=65)
   ) {isSetParameters=true}
  ]
  rendererClasses [&NiagaraMeshRendererProperties, &NiagaraSpriteRendererProperties]
  renderers [
   {rendererClass=&NiagaraMeshRendererProperties, rendererIndex=0}
   {rendererClass=&NiagaraMeshRendererProperties, rendererIndex=1}
   {rendererClass=&NiagaraMeshRendererProperties, rendererIndex=2}
   {rendererClass=&NiagaraSpriteRendererProperties, rendererIndex=3}
  ]
 }
 Trail_AudioReactive {
  emitterUpdateScript [
   &EmitterState(Life Cycle Mode=::NewEnumerator1, Loop Behavior=::NewEnumerator1, Loop Duration Mode=::NewEnumerator1, UseLoopDelay=false)
   &SpawnBurst_Instantaneous(Spawn Count=8, Use Spawn Probability=false, UseLoopCountLimit=false)
  ]
  particleSpawnScript [
   &PartitionParticles()
   &InitializeParticle(Color Mode=::NewEnumerator0, Sprite Size=<5 70>, Sprite Size Mode=::NewEnumerator4, UsePositionOffset=false, Write Material Random=true)
   /SetVariables(Particles.AudioPosition:NiagaraFloat=&ReturnNormalizedExecIndex()) {isSetParameters=true}
  ]
  particleUpdateScript [
   &ParticleState(Kill Particles When Lifetime Has Elapsed=false)
   &SkeletalMeshLocation(
    Bone / Socket Index=$Particles.PartitionParticles.Partition
    Bone Sampling Mode=::NewEnumerator4
    MaskSpawnGroup=false
    Orientation Sampling=::NewEnumerator1
    Sampled Position Offset=&Multiply_VectorByFloat(
     Vector=&MakeVector(
      X=&FloatFromCurve(
       FloatCurve=DataInterface() {propertyValues {Curve {keys [<at 0 1>,<at 1 0>]}}}
       CurveIndex=&ReturnNormalizedExecIndex()
       Scale Curve=-20
      )
      Y=1
      Z=1
     )
     Float=&Lerp_Float(A=-1, B=1, Alpha=&MakeFloatFromInt(Int=$Particles.PartitionParticles.Partition))
    )
    Skeletal Mesh=$System.Skeletal Mesh
   )
   &UpdateMeshOrientation(
    Facing Direction=$System.FN_Contrail_Setup.Direction Vector
    Facing Interpolation=false
    Orientation Method=::NewEnumerator1
    Override Quat=false
    Stabilization Mode:ENiagara_OrientVectorCount=::NewEnumerator2
    TransformFacing=false
    TransformSide=false
   )
   &AlignSpriteToMeshOrientation(Mesh Orientation Relative Sprite Alignment Vector=<-1 0 0>, SetAlignment=true, SetFacing=false)
   /SetVariables(Particles.AudioSample:NiagaraFloat=&Add_Float(A=0, B=0)) {isSetParameters=true}
   &ScaleSpriteSize(
    Scale Factor=&MakeCustomVector2DFromBool(
     Bool=&SetBoolByFloatComparison(A=$System.AudioScale, B=0.1, Comparison Type=::NewEnumerator2)
     True Vector2D=&Multiply_Vector2D(
      A=&Lerp_Vector2(
       A=<1 0.5>
       B=<1 1.5>
       Alpha=&Multiply_Float(
        A=&Sine(Normalized Angle=&Add_Float(A=$Emitter.Age, B=$Particles.AudioPosition), Period=1, Scale=1.2, Bias=0)
        B=$System.AudioScale
       )
      )
      B=&Lerp_Vector2(A=<1 1>, B=<1 2>, Alpha=$System.FN_Contrail_Setup.Dive Value)
     )
     False Vector2D=&MakeVector2D(
      X=1
      Y=&Sine(
       Normalized Angle=&Add_Float(A=$Emitter.Age, B=$Particles.AudioPosition)
       Period=&Lerp_Float(A=4, B=3, Alpha=$System.FN_Contrail_Setup.Dive Value)
       Scale=&Lerp_Float(A=1.3, B=2.2, Alpha=$System.FN_Contrail_Setup.Dive Value)
       Bias=0
      )
     )
    )
    Scale Sprite Size Mode=::NewEnumerator2
   ) {module=ScaleSpriteSize003}
   &ScaleColor(
    Scale Alpha=&FloatFromCurve(
     FloatCurve=DataInterface() {propertyValues {Curve {keys [
         {interpMode=RCIM_Cubic, tangentMode=RCTM_User, time=0, value=0}
         {arriveTangent=9.1187, interpMode=RCIM_Cubic, leaveTangent=9.1187, tangentMode=RCTM_User, time=0.3, value=1}
        ]}}}
     CurveIndex=$Emitter.Age
    )
    Scale RGB=&Multiply_Vector(
     A=&MakeVectorFromLinearColorRGB(Color=&ColorFromCurve(
       ColorCurve=DataInterface() {propertyValues {
         AlphaCurve {keys [<at 0 1 RCIM_Cubic>]}
         BlueCurve {keys [<at 0 0.807 RCIM_Cubic>,<at 0.2246 0>,<at 0.3968 0.019>,<at 0.6058 0.8879>,<at 0.8235 0.8879>,<at 1 0.807 RCIM_Cubic>]}
         GreenCurve {keys [<at 0 0.8879 RCIM_Cubic>,<at 0.2246 1>,<at 0.3968 1>,<at 0.6058 0.3043>,<at 0.8235 0>,<at 1 0.8879 RCIM_Cubic>]}
         RedCurve {keys [<at 0 0.0052 RCIM_Cubic>,<at 0.2246 0.9451>,<at 0.3968 0.1838>,<at 0.6058 0>,<at 0.8235 0.4788>,<at 1 0.0052 RCIM_Cubic>]}
        }}
       CurveIndex=&Multiply_Float(
        A=&Sine(Normalized Angle=&Add_Float(A=$Emitter.Age, B=$Particles.AudioPosition), Period=2, Scale=1, Bias=0)
        B=$System.AudioScale
       )
      ))
     B=&VectorFromFloat(Value=0.65)
    )
    ScaleA=true
    ScaleRGB=true
   )
  ]
  rendererClasses [&NiagaraSpriteRendererProperties]
  renderers [{rendererClass=&NiagaraSpriteRendererProperties}]
 }
 Trail_Circuits001 {
  emitterUpdateScript [
   &EmitterState(Life Cycle Mode=::NewEnumerator1, Loop Duration=0, UseLoopDelay=false)
   &SpawnRate(SpawnRate=&RandomRangeFloat(Minimum=2, Maximum=2), Use Spawn Probability=false)
  ]
  particleSpawnScript [
   &PartitionParticles()
   &InitializeParticle(
    Color=&SelectLinearColorFromArray(
     Array Sampling Mode=::NewEnumerator1
     Direct Array Index=&RandomRangeInt(Minimum=0, Maximum=4)
     Color Selection Array=DataInterface() {propertyValues {ColorData [<rgba 3 0.0961 2.9812 1>, <rgba 2 1.6172 0.0553 1>, <rgba 0.0961 2.0591 3 1>, <rgba 0.7086 0.07 3 1>, <rgba 0.1148 2 0.1075 1>]}}
     Random Seed=?
    )
    Lifetime Max=1.5
    Lifetime Min=2
    Lifetime Mode=::NewEnumerator1
    Sprite Size=&RandomRangeVector2D(Minimum=<8 90>, Maximum=<12 170>)
    Sprite Size Mode=::NewEnumerator4
    UsePositionOffset=false
    Write Material Random=true
   )
  ]
  particleUpdateScript [
   &ParticleState()
   &SkeletalMeshLocation(
    Bone / Socket Index=$Particles.PartitionParticles.Partition
    Bone Sampling Mode=::NewEnumerator4
    MaskSpawnGroup=false
    Orientation Sampling=::NewEnumerator1
    Sampled Position Offset=&Multiply_VectorByFloat(
     Vector=&RandomRangeVector(Minimum=<15 -1 0>, Maximum=<-30 1 0>)
     Float=&Lerp_Float(A=-1, B=1, Alpha=&MakeFloatFromInt(Int=$Particles.PartitionParticles.Partition))
    )
    Skeletal Mesh=$System.Skeletal Mesh
   ) {module=SkeletalMeshLocation002}
   /SetVariables(
    Particles.GlowColor=&MultiplyLinearColor(Linear Color=$StackContext.Color, Scale Factor=&MakeVector4(X=1, Y=1, Z=1, W=0.5))
    Particles.GlowPosition:NiagaraPosition=&AddVectorToPosition(Position=$StackContext.Position, Vector=<0 0 5>)
    Particles.GlowScale:Vector2f=<33 33>
   ) {isSetParameters=true}
   &UpdateMeshOrientation(
    Facing Direction=$System.FN_Contrail_Setup.Direction Vector
    Facing Interpolation=false
    Orientation Method=::NewEnumerator1
    Override Quat=false
    Stabilization Mode:ENiagara_OrientVectorCount=::NewEnumerator2
    TransformFacing=false
    TransformSide=false
   )
   &AlignSpriteToMeshOrientation(Mesh Orientation Relative Sprite Alignment Vector=<-1 0 0>, SetAlignment=true, SetFacing=false)
   &CameraOffset(Camera Offset Amount=&RandomRangeFloat(Minimum=0, Maximum=5))
   &ScaleSpriteSize(
    Scale Factor=&Multiply_Vector2D(
     A=&MakeVector2D(X=1, Y=1)
     B=&Lerp_Vector2(A=<1 1>, B=<1 1.5>, Alpha=$System.FN_Contrail_Setup.Dive Value)
    )
    Scale Sprite Size Mode=::NewEnumerator2
   ) {module=ScaleSpriteSize001}
   &ScaleColor(
    Scale Alpha=&FloatFromCurve(FloatCurve=DataInterface() {propertyValues {Curve {keys [
         <at 0 0 RCIM_Cubic>
         {arriveTangent=1.25, interpMode=RCIM_Cubic, leaveTangent=0.0529, tangentMode=RCTM_Break, time=0.15, value=1}
         {interpMode=RCIM_Cubic, tangentMode=RCTM_User, time=0.8, value=1}
         {interpMode=RCIM_Cubic, tangentMode=RCTM_User, time=1, value=0}
        ]}}})
    ScaleA=true
    ScaleRGB=false
   )
   &DynamicMaterialParameters(
    Index 0 Param 1=&UniformAOrBFloat(A=0, Evaluation Type=::NewEnumerator0, Recalculate Random Each Loop=true, Override Seed=false)
    Index 0 Param 2=&UniformAOrBFloat(A=0, Distribution Weight=0.3, Evaluation Type=::NewEnumerator0, Recalculate Random Each Loop=true, Override Seed=false)
    Module.Index0Param0Write=true
    Module.Index0Param1Write=true
    Module.Index0Param2Write=true
    Module.Index0Param3Write=true
   )
  ]
  rendererClasses [&NiagaraSpriteRendererProperties]
  renderers [
   {rendererClass=&NiagaraSpriteRendererProperties}
   {rendererClass=&NiagaraSpriteRendererProperties, rendererIndex=1}
  ]
 }
 Trail_Main_Arms_Legs001 {
  emitterUpdateScript [
   &EmitterState(Life Cycle Mode=::NewEnumerator1, Loop Behavior=::NewEnumerator1, Loop Duration Mode=::NewEnumerator1, UseLoopDelay=false)
   &SpawnBurst_Instantaneous(Spawn Count=4, Use Spawn Probability=false, UseLoopCountLimit=false)
  ]
  particleSpawnScript [
   &PartitionParticles(Number Of Partitions=4)
   /SetVariables(Particles.SetGroups:NiagaraInt32=&SelectIntFromArray(Direct Array Index=$Particles.PartitionParticles.Partition, Int Selection Array=DataInterface() {propertyValues {IntData [0,0,1,1]}})) {isSetParameters=true}
   &InitializeParticle(
    Color=<rgba 0 1 0.0484 1>
    Sprite Size=&Lerp_Vector2(A=<35 120>, B=<12 50>, Alpha=&MakeFloatFromInt(Int=$Particles.SetGroups))
    Sprite Size Mode=::NewEnumerator4
    UsePositionOffset=false
    Write Material Random=true
   )
  ]
  particleUpdateScript [
   &ParticleState(
    Kill Particles When Lifetime Has Elapsed=false
    Let Infinitely Lived Particles Die When Emitter Deactivates=true
    Loop Particles Lifetime=true
   )
   &SkeletalMeshLocation(
    Bone / Socket Index=&SelectIntFromArray(Direct Array Index=$Particles.PartitionParticles.Partition, Int Selection Array=DataInterface() {propertyValues {IntData [0,1,4,5]}})
    Bone Sampling Mode=::NewEnumerator4
    MaskSpawnGroup=false
    Orientation Sampling=::NewEnumerator1
    Sampled Position Offset=&Lerp_Vector(
     A=&Multiply_VectorByFloat(
      Vector=&RandomRangeVector(Minimum=<0 0 0>, Maximum=<-20 0 0>)
      Float=&Lerp_Float(A=-1, B=1, Alpha=&MakeFloatFromInt(Int=$Particles.PartitionParticles.Partition))
     )
     B=<1 1 1>
     Alpha=&MakeFloatFromInt(Int=$Particles.SetGroups)
    )
    Skeletal Mesh=$System.Skeletal Mesh
   ) {module=SkeletalMeshLocation001}
   &ScaleSpriteSize(
    Scale Factor=&Multiply_Vector2D(
     A=&Lerp_Vector2(A=<1 1.4>, B=<1 1.8>, Alpha=$System.FN_Contrail_Setup.Dive Value)
     B=&MakeVector2D(X=1, Y=1)
    )
    Scale Sprite Size Mode=::NewEnumerator2
   )
   &UpdateMeshOrientation(
    Facing Direction=$System.FN_Contrail_Setup.Direction Vector
    Facing Interpolation=false
    Orientation Method=::NewEnumerator1
    Override Quat=false
    Stabilization Mode:ENiagara_OrientVectorCount=::NewEnumerator2
    TransformFacing=false
    TransformSide=false
   )
   &AlignSpriteToMeshOrientation(Mesh Orientation Relative Sprite Alignment Vector=<-1 0 0>, SetAlignment=true, SetFacing=false)
   &DynamicMaterialParameters(
    Index 0 Param 1=&Lerp_Float(A=&RandomRangeFloat(Minimum=1.5, Maximum=2), B=3, Alpha=$System.FN_Contrail_Setup.Dive Value)
    Index 0 Param 2=&RandomRangeFloat(Minimum=1, Maximum=2)
    Module.Index0Param0Write=true
    Module.Index0Param1Write=true
    Module.Index0Param2Write=true
    Module.Index0Param3Write=true
   )
   &ScaleColor(
    Scale Alpha=&FloatFromCurve(
     FloatCurve=DataInterface() {propertyValues {Curve {keys [
         {interpMode=RCIM_Cubic, tangentMode=RCTM_User, time=0, value=0}
         {arriveTangent=9.1187, interpMode=RCIM_Cubic, leaveTangent=9.1187, tangentMode=RCTM_User, time=0.3, value=1}
        ]}}}
     CurveIndex=$Emitter.Age
    )
    ScaleA=true
    ScaleRGB=false
   )
  ]
  rendererClasses [&NiagaraSpriteRendererProperties]
  renderers [{rendererClass=&NiagaraSpriteRendererProperties}]
 }
 squares001 {
  emitterUpdateScript [
   &EmitterState(Life Cycle Mode=::NewEnumerator1, Loop Behavior=::NewEnumerator1, Loop Duration Mode=::NewEnumerator1, UseLoopDelay=false)
   &SpawnBurst_Instantaneous(Spawn Count=4, Use Spawn Probability=false, UseLoopCountLimit=false)
   &SpawnBurst_Instantaneous(Spawn Count=2, Spawn Group=1, Use Spawn Probability=false, UseLoopCountLimit=false) {module=SpawnBurst_Instantaneous001}
  ]
  particleSpawnScript [
   /SetVariables(Particles.UpdateGroup:NiagaraInt32=$Emitter.SpawnGroup) {isSetParameters=true}
   &PartitionParticles()
   &InitializeParticle(
    Color=&LerpLinearColors(
     StartColor=&SelectLinearColorFromArray(
      Array Sampling Mode=::NewEnumerator1
      Direct Array Index=$Particles.UniqueID
      Color Selection Array=DataInterface() {propertyValues {ColorData [<rgba 1 0 0.8627 1>, <rgba 0 1 0.0767 1>, <rgba 0 0.3804 1 1>, <rgba 1 0.9256 0 1>]}}
      Random Seed=?
     )
     EndColor=<rgba 1 1 1 1>
     LerpFactor=&MakeFloatFromInt(Int=$Particles.UpdateGroup)
    )
    Lifetime=2
    Sprite Size Mode=::NewEnumerator3
    Uniform Sprite Size=&Multiply_Float(
     A=&Lerp_Float(
      A=&SelectFloatFromArray(Direct Array Index=$Particles.UniqueID, Float Selection Array=DataInterface() {propertyValues {FloatData [50,50,43,43]}})
      B=30
      Alpha=&MakeFloatFromInt(Int=$Emitter.SpawnGroup)
     )
     B=0.8
    )
    UsePositionOffset=false
    Write Material Random=true
   )
   /SetVariables(Particles.AudioPosition:NiagaraFloat=&ReturnNormalizedExecIndex()) {isSetParameters=true}
  ]
  particleUpdateScript [
   &ParticleState(
    Kill Particles When Lifetime Has Elapsed=false
    Let Infinitely Lived Particles Die When Emitter Deactivates=true
    Loop Particles Lifetime=true
   )
   &ScaleSpriteSize(
    Scale Factor=&Vector2DFromFloat(Value=&Lerp_Float(A=1, B=1.4, Alpha=$System.AudioBassScale))
    Scale Sprite Size Mode=::NewEnumerator2
   ) {module=ScaleSpriteSize001}
   &SpriteRotationRate(Rotation Rate=&Lerp_Float(
     A=&Multiply_Float(
      A=&Lerp_Float(A=22, B=99, Alpha=&MakeFloatFromInt(Int=$Particles.UpdateGroup))
      B=&Lerp_Float(A=1, B=-1, Alpha=&MakeFloatFromInt(Int=$Particles.PartitionParticles.Partition))
     )
     B=&Multiply_Float(
      A=&Lerp_Float(A=500, B=800, Alpha=&MakeFloatFromInt(Int=$Particles.UpdateGroup))
      B=&Lerp_Float(A=1, B=-1, Alpha=&MakeFloatFromInt(Int=$Particles.PartitionParticles.Partition))
     )
     Alpha=$System.FN_Contrail_Setup.Dive Value
    )) {module=SpriteRotationRate001}
   &SkeletalMeshLocation(
    Bone / Socket Index=&SelectIntFromArray(Direct Array Index=$Particles.PartitionParticles.Partition, Int Selection Array=DataInterface() {propertyValues {IntData [6,7]}})
    Bone Sampling Mode=::NewEnumerator4
    MaskSpawnGroup=false
    Sampled Position Offset Coordinate Space:ENiagaraSkelMeshTransforms=::NewEnumerator0
    Skeletal Mesh=$System.Skeletal Mesh
   ) {module=SkeletalMeshLocation001}
   /SetVariables(Particles.Position=&AddVectorToPosition(
     Position=$StackContext.Position
     Vector=&Lerp_Vector(
      A=<0 0 0>
      B=&MakeVector(
       X=0
       Y=&FloatFromCurve(
        FloatCurve=DataInterface() {propertyValues {Curve {keys [<at 0 1>,<at 1 0>]}}}
        CurveIndex=&ReturnNormalizedExecIndex()
        Scale Curve=5
       )
       Z=&FloatFromCurve(
        FloatCurve=DataInterface() {propertyValues {Curve {keys [<at 0 1>,<at 1 0>]}}}
        CurveIndex=&ReturnNormalizedExecIndex()
        Scale Curve=40
       )
      )
      Alpha=$System.FN_Contrail_Setup.Dive Value
     )
    )) {isSetParameters=true}
   &SolveForcesAndVelocity(Clamp Velocity=false, Limit Acceleration=false)
   /SetVariables(Particles.AudioWaveform:NiagaraFloat=&Waveform("Waveform [1]"=::NewEnumerator0, "[1] Frequency"=1, "[1] Amplitude Min/Max"=<0 1>, "[1] Scale"=2)) {isSetParameters=true}
   &DynamicMaterialParameters(
    Index 0 Param 1=&Lerp_Float(
     A=&Lerp_Float(
      A=&SelectFloatFromArray(Direct Array Index=$Particles.UniqueID, Float Selection Array=DataInterface() {propertyValues {FloatData [0.01,0.01,0.03,0.03]}})
      B=&SelectFloatFromArray(Direct Array Index=$Particles.UniqueID, Float Selection Array=DataInterface() {propertyValues {FloatData [0.03,0.03,0.06,0.06]}})
      Alpha=&Multiply_Float(A=$Particles.AudioWaveform, B=$System.AudioBassScale)
     )
     B=&Lerp_Float(A=0.1, B=0.15, Alpha=&Multiply_Float(A=$Particles.AudioWaveform, B=$System.AudioBassScale))
     Alpha=&MakeFloatFromInt(Int=$Particles.UpdateGroup)
    )
    Index 0 Param 2=&Lerp_Float(A=0, B=1, Alpha=&MakeFloatFromInt(Int=$Particles.UpdateGroup))
    Module.Index0Param0Write=true
    Module.Index0Param1Write=true
    Module.Index0Param2Write=true
    Module.Index0Param3Write=true
   ) {module=DynamicMaterialParameters004}
  ]
  rendererClasses [&NiagaraSpriteRendererProperties]
  renderers [{rendererClass=&NiagaraSpriteRendererProperties}]
 }
}
summary {
 systemName=NS_Contrail_JadeTowel
 emitters [
  {emitterName=Trail_Circuits001, rendererClasses [&NiagaraSpriteRendererProperties]}
  {emitterName=Light001, rendererClasses [&NiagaraLightRendererProperties, &NiagaraSpriteRendererProperties]}
  {emitterName=squares001, rendererClasses [&NiagaraSpriteRendererProperties]}
  {emitterName=StageLights, rendererClasses [&NiagaraMeshRendererProperties, &NiagaraSpriteRendererProperties]}
  {emitterName=Glow_Body001, rendererClasses [&NiagaraMeshRendererProperties]}
  {emitterName=Sparkles_GPU, simTarget=GPUComputeSim, rendererClasses [&NiagaraSpriteRendererProperties]}
  {emitterName=Trail_AudioReactive, rendererClasses [&NiagaraSpriteRendererProperties]}
  {emitterName=Trail_Main_Arms_Legs001, rendererClasses [&NiagaraSpriteRendererProperties]}
  {emitterName=Sparkles, rendererClasses [&NiagaraSpriteRendererProperties]}
  {emitterName=Glowsticks001, rendererClasses [&NiagaraMeshRendererProperties]}
 ]
 userVariables [
  User.bIsFrontEnd {type=NiagaraBool}
  User.bIsFrontEndPreview {type=NiagaraBool, defaultValue=true}
  User.bIsLocalPlayer {type=NiagaraBool, defaultValue=true}
  User.ParaGlideLeanAlpha {type=NiagaraFloat, defaultValue=0, description=""}
 ]
}
